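; Lock-on ranged weapon (Lotus Mustache) that fires homing projectiles and shows the locked player's health.
; The locally controlled user spawns a widget actor on equip; non-server Tick pins it to the target's spine and refreshes health through Server/Client-named events.
; With a target, projectiles home without bouncing and ignore world collision, with turn ramp set by distance.
; Without one, they get 0.5 gravity and lose bounce unless bandage spawning is enabled.
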(event Custom|SetHealthPercentServer
  (bind _asfort_player_pawn_athena (Utilities|Casting|CastToFortPlayerPawnAthena (Variables|Default|GetLockOnTarget)))
  (CallFunction|SetHealthPercentClient (/ (FortPawn|GetHealth _asfort_player_pawn_athena) (FortPawn|GetMaxHealth _asfort_player_pawn_athena))))

(event Custom|SetHealthPercentClient
  (params (HealthPercent :type Float%20%28double-precision%29))
  (Progress|SetPercent (Variables|/Game/Athena/Items/Gameplay/Lotus/Mustache/Lotus_Mustache_HealthBar.Lotus_Mustache_HealthBar_C|GetProgressBar_35 (Variables|Default|GetLockOnWidget)) HealthPercent)
  (Variables|Default|SetHealthPercent HealthPercent))

(event Custom|NewTarget)

(event FortWeapon|EventOnEquippedAndAttached
  (bind _instigator (Variables|Actor|GetInstigator))
  (if (Pawn|IsLocallyControlled _instigator)
    (bind _returnvalue (Game|SpawnActorfromClass "/Game/Athena/Items/Gameplay/Lotus/Mustache/Athena_LockOnWidgetActor_Proto.Athena_LockOnWidgetActor_Proto_C" (Math|Conversions|ToTransform(Vector) (Transformation|GetActorLocation)) "Undefined" "OverrideRootScale" 0 _instigator))
    (Variables|Default|SetWidgetActor _returnvalue)
    (bind _aslotus_mustache_health_bar (Utilities|Casting|CastToLotus_Mustache_HealthBar (UserInterface|GetUserWidgetObject (Variables|/Game/Athena/Items/Gameplay/Lotus/Mustache/Athena_LockOnWidgetActor_Proto.Athena_LockOnWidgetActor_Proto_C|GetHealthBar _returnvalue))))
    (Variables|Default|SetLockOnWidget _aslotus_mustache_health_bar)
    (Rendering|SetActorHiddenInGame (Variables|Default|GetWidgetActor) true)))

(event FortWeapon|EventOnLocalUnEquip
  (Actor|DestroyActor (stale FortWeapon|EventOnEquippedAndAttached _returnvalue)))

(event EventTick
  (params (DeltaSeconds :type Float%20%28single-precision%29))
  (bind _lockontarget (Variables|Default|GetLockOnTarget))
  (if (not (Networking|IsServer))
    (Utilities|IsValid (Variables|Default|GetWidgetActor)
      (:Is%20Valid
        (Utilities|IsValid _lockontarget
          (:Is%20Valid
            (bind _asfort_player_pawn (Utilities|Casting|CastToFortPlayerPawn _lockontarget))
            (Transformation|SetActorLocation (Variables|Default|GetWidgetActor) (Transformation|GetSocketLocation (Variables|/Script/Engine.Character|GetMesh _asfort_player_pawn) "spine_04") false false)
            (CallFunction|SetHealthPercentServer)
            (Rendering|SetActorHiddenInGame (Variables|Default|GetWidgetActor) false))
          (:Is%20Not%20Valid
            (Rendering|SetActorHiddenInGame (Variables|Default|GetWidgetActor) true)))))))

(event FortWeapon|EventOnSpawnProjectile
  (params (SpawnProjectile :type Fort%20Projectile%20Base%20Object%20Reference))
  (bind _homingdistancethreshold (Variables|Default|GetHomingDistanceThreshold))
  (bind _capsulecomponent (Variables|/Script/FortniteGame.FortProjectileBase|GetCapsuleComponent SpawnProjectile))
  (Utilities|IsValid SpawnProjectile
    (:Is%20Valid
      (Utilities|IsValid (Variables|Default|GetLockOnTarget)
        (:Is%20Valid
          (Variables|/Script/Engine.ProjectileMovementComponent|SetbShouldBounce false (Variables|/Script/FortniteGame.FortProjectileBase|GetProjectileMovementComponent SpawnProjectile))
          (Collision|SetCollisionResponsetoChannel _capsulecomponent "ECC_WorldStatic" "ECR_Ignore")
          (Collision|SetCollisionResponsetoChannel _capsulecomponent "ECC_WorldDynamic" "ECR_Ignore")
          (Projectile|SetHomingTurnSpeedMin SpawnProjectile SpawnProjectile 0.0 :!class /Script/FortniteGame.FortProjectileMovementInterface)
          (Projectile|SetHomingTurnSpeedMax SpawnProjectile SpawnProjectile (Variables|Default|GetTurnSpeedMax) :!class /Script/FortniteGame.FortProjectileMovementInterface)
          (Projectile|SetHomingLockTargetDistanceThreshold SpawnProjectile SpawnProjectile _homingdistancethreshold :!class /Script/FortniteGame.FortProjectileMovementInterface)
          (Projectile|SetHomingTimeUntilMaxTurnSpeed SpawnProjectile SpawnProjectile (Math|Float|SelectFloat 0.0 1.0 (<= (Transformation|GetDistanceTo (Variables|Default|GetLockOnTarget) self) _homingdistancethreshold)) :!class /Script/FortniteGame.FortProjectileMovementInterface)
          (Projectile|SetHomingStyle SpawnProjectile SpawnProjectile "LockOn" false :!class /Script/FortniteGame.FortProjectileMovementInterface)
          (Projectile|SetHomingTarget SpawnProjectile SpawnProjectile (Variables|Default|GetLockOnTarget) 0 :!class /Script/FortniteGame.FortProjectileMovementInterface))
        (:Is%20Not%20Valid
          (if (not (Class|AthenaFunctionLibrary|RowtoBool (Variables|Default|GetShouldSpawnBandages)))
            (Variables|/Script/Engine.ProjectileMovementComponent|SetbShouldBounce false (Variables|/Script/FortniteGame.FortProjectileBase|GetProjectileMovementComponent SpawnProjectile)))
          (Projectile|SetGravityScale SpawnProjectile 0.5))))))

(event EventOnPlayWeaponFireFX
  (params (bPersistentFire :type Boolean) (bSecondaryFire :type Boolean))
  ;[1 0.9704 0] Fire!
  (|CallParentFunction bPersistentFire bSecondaryFire)
  (if (Variables|VFX|GetCanPlay?)
    (CallFunction|MuzzleFlashFX false)))

(fn ConstructionScript
  (|CallParentFunction))
