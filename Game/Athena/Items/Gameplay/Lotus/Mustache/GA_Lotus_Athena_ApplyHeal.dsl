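; GA_Lotus_Athena_ApplyHeal is a gameplay ability that gives the Fort player pawn avatar a burst heal from the Lotus Mustache item.
; On activation it casts the avatar to a player pawn and checks whether health is below maximum.
; If so, the instigator from the granting effect context applies GE_Lotus_Mustache_Heal_Burst to the avatar, passing that context along.
; The ability then ends immediately.
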(event Ability|EventActivateAbility
  (bind _asfort_player_pawn (Utilities|Casting|CastToFortPlayerPawn (Ability|GetAvatarActorfromActorInfo)))
  (if (not (>= (/ (FortPawn|GetHealth _asfort_player_pawn) (FortPawn|GetMaxHealth _asfort_player_pawn)) 1.0))
    (GameplayEffects|FortApplyGameplayEffectToTarget (Ability|GetFortAbilitySystemComponent (Ability|EffectContext|GetInstigatorActor (Ability|GetGrantedbyEffectContext))) "/Game/Athena/Items/Gameplay/Lotus/Mustache/GE_Lotus_Mustache_Heal_Burst.GE_Lotus_Mustache_Heal_Burst_C" (Ability|GetFortAbilitySystemComponent (Ability|GetAvatarActorfromActorInfo)) 0 0 0 0.0 (Ability|GetGrantedbyEffectContext)))
  (Ability|EndAbility))

(event Ability|EventOnEndAbility
  (params (bWasCancelled :type Boolean)))

(event Ability|EventFailedToActivatePassiveAbility
  (params (ActorInfo :type Gameplay%20Ability%20Actor%20Info%20Structure))
  (|CallParentFunction))
