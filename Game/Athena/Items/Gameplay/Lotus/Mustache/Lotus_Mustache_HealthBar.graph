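# Widget Blueprint for the Lotus Mustache item's health bar, showing a single green progress bar.
# A centre-anchored bar on a canvas panel uses the player health bar material as its fill, with alignment lifting it above centre.
# The SetHealth custom event writes its HealthPercent input straight to the bar.
# Construct events are empty, the timer nodes have no execution link, and the Target actor variable is unused.
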WidgetBlueprint /Game/Athena/Items/Gameplay/Lotus/Mustache/Lotus_Mustache_HealthBar
graphs [EventGraph {nodes [
   n1 = SetHealth!() -> (OutputDelegate, then, HealthPercent:double=~) {is=title K2Node_CustomEvent}
   > SetPercent!(
    execute
    self:Progress Bar Object Reference=Get ProgressBar_35() -> (ProgressBar_35:Progress Bar Object Reference)
    InPercent:float=@n1.HealthPercent
   ) -> (then)
   Clear and Invalidate Timer by Handle(Handle=Set Timer by Event(Time=0.02, bLooping=true))
   Event PreConstruct()
   Event Construct()
   Event Destruct()
  ]}]
variables [Target {type=Actor, default=null}]
widgets {
 info {inheritedWidgetCount=0, parentClass=&UserWidget, rootWidgetClass=&CanvasPanel, widgetCount=2}
 widgets [
  CanvasPanel_29 {widgetClassPath=&CanvasPanel, widget=CanvasPanel_29}
  ProgressBar_35 {
   widgetClassPath=/Script/UMG.ProgressBar
   bIsVariable=true
   parent=CanvasPanel_29
   widget=ProgressBar_35
   properties {
    fillColorAndOpacity=<rgba 0 1 0.0449 1>
    widgetStyle {
     enableFillAnimation=false
     backgroundImage {
      bIsDynamicallyLoaded=false
      drawAs=Box
      imageType=FullColor
      mirroring=NoMirror
      resourceName=../../../Engine/Content/Slate/Common/ProgressBar_Background.png
      resourceObject=null
      tiling=NoTile
      imageSize=<0 0>
      margin=<ltrb 0.4167 0.4167 0.4167 0.4167>
      outlineSettings {
       bUseBrushTransparency=false
       roundingType=HalfHeightRadius
       width=0
       color {colorUseRule=UseColor_Specified, specifiedColor=<rgba 0 0 0 0>}
       cornerRadii=<0 0 0 0>
      }
      tintColor {colorUseRule=UseColor_Specified, specifiedColor=<rgba 0.7715 0.8789 0.9375 1>}
      uVRegion {bIsValid=false, max=<0 0>, min=<0 0>}
     }
     fillImage {
      bIsDynamicallyLoaded=false
      drawAs=Box
      imageType=FullColor
      mirroring=NoMirror
      resourceName=../../../Engine/Content/Slate/Common/ProgressBar_Fill.png
      resourceObject=/Game/Athena/HUD/Team/MI_Player_HealthBar
      tiling=NoTile
      imageSize=<0 0>
      margin=<ltrb 0.4167 0.4167 0.4167 0.4167>
      outlineSettings {
       bUseBrushTransparency=false
       roundingType=HalfHeightRadius
       width=0
       color {colorUseRule=UseColor_Specified, specifiedColor=<rgba 0 0 0 0>}
       cornerRadii=<0 0 0 0>
      }
      tintColor {colorUseRule=UseColor_Specified, specifiedColor=<rgba 1 1 1 1>}
      uVRegion {bIsValid=false, max=<0 0>, min=<0 0>}
     }
     marqueeImage {
      bIsDynamicallyLoaded=false
      drawAs=Image
      imageType=FullColor
      mirroring=NoMirror
      resourceName=../../../Engine/Content/Slate/Common/ProgressBar_Marquee.png
      resourceObject=null
      tiling=Horizontal
      imageSize=<20 12>
      margin=<ltrb 0 0 0 0>
      outlineSettings {
       bUseBrushTransparency=false
       roundingType=HalfHeightRadius
       width=0
       color {colorUseRule=UseColor_Specified, specifiedColor=<rgba 0 0 0 0>}
       cornerRadii=<0 0 0 0>
      }
      tintColor {colorUseRule=UseColor_Specified, specifiedColor=<rgba 1 1 1 1>}
      uVRegion {bIsValid=false, max=<0 0>, min=<0 0>}
     }
    }
   }
   slot {
    class=CanvasPanelSlot
    properties {layoutData {alignment=<0.5 6>, anchors {maximum=<0.5 0.5>, minimum=<0.5 0.5>}, offsets=<ltrb -0.3819 0 100 10>}}
   }
  }
 ]
}
