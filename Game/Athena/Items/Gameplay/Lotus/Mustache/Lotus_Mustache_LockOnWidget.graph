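# Lock-on cursor widget for the Lotus Mustache gameplay item, showing a small marker at the center of its canvas.
# A CanvasPanel root holds a single Image, center-anchored and center-aligned, that draws the M_Athena_Mustache_LockOnCursor material.
# The event graph is empty and no variables or animation logic are defined here, so the look comes from that material.
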WidgetBlueprint /Game/Athena/Items/Gameplay/Lotus/Mustache/Lotus_Mustache_LockOnWidget
graphs [EventGraph {links [], nodes []}]
variables []
widgets {
 info {inheritedWidgetCount=0, parentClass=&UserWidget, rootWidgetClass=&CanvasPanel, widgetCount=2}
 widgets [
  CanvasPanel_29 {widgetClassPath=&CanvasPanel, widget=CanvasPanel_29}
  Image_83 {
   widgetClassPath=&Image
   bIsVariable=true
   parent=CanvasPanel_29
   widget=Image_83
   properties {brush {
     bIsDynamicallyLoaded=false
     drawAs=Image
     imageType=NoImage
     mirroring=NoMirror
     resourceName=None
     resourceObject=&/M_Athena_Mustache_LockOnCursor
     tiling=NoTile
     imageSize=<32 32>
     margin=<ltrb 0 0 0 0>
     outlineSettings {
      bUseBrushTransparency=false
      roundingType=HalfHeightRadius
      width=0
      color {colorUseRule=UseColor_Specified, specifiedColor=<rgba 0 0 0 0>}
      cornerRadii=<0 0 0 0>
     }
     tintColor {colorUseRule=UseColor_Specified, specifiedColor=<rgba 1 1 1 1>}
     uVRegion {bIsValid=false, max=<0 0>, min=<0 0>}
    }}
   slot {
    class=CanvasPanelSlot
    properties {layoutData {alignment=<0.5 0.5>, anchors {maximum=<0.5 0.5>, minimum=<0.5 0.5>}, offsets=<ltrb 0 0 25 25>}}
   }
  }
 ]
}
