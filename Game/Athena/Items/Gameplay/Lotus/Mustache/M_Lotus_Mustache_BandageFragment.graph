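# Masked two-sided lit material for Lotus Mustache bandage fragment effects, textured with the medical bandage color and flattened normal.
# UseSplineMeshAsTrail switches between a spline trail and a particle look with Fresnel glow.
# The trail offsets vertices with panning ripple noise, a timed eased pull along BulletOrigin minus actor position, and SplineThicken, with noise eroding its mask.
# Otherwise, Perlin noise wobbles vertices and dynamic parameters drive noise-based erosion.
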Material /Game/Athena/Items/Gameplay/Lotus/Mustache/M_Lotus_Mustache_BandageFragment
graph {
 properties {bUsedWithMeshParticles=true, blendMode=BLEND_Masked, pixelDepthOffsetMode=PDOM_Legacy, twoSided=true, usedShadingModels=Default Lit}
 nodes [
  n1 = MaterialFunctionCall(Float2=MaterialFunctionCall(XY=TextureCoordinate() {uTiling=0.5}) {materialFunction=&Swizzle}.YX) {materialFunction=&BreakOutFloat2Components}
  n2 = ScalarParameter() {defaultValue=1, parameterName=U_Tile}
  n3 = Panner(
   Coordinate=AppendVector(A=Multiply(A=@n1.R, B=@n2), B=@n1.G)
   Time=Add(A=Time(), B=DynamicParameter() {paramNames [RandomInitailU, MaskCutUVRotator, RandomTwistOffset, OffsetTwistSpeed]}.RandomInitailU)
   Speed=AppendVector(A=ScalarParameter() {defaultValue=1, parameterName=WPO_PanSpeed_U}, B=ScalarParameter() {defaultValue=0.1, parameterName=WPO_PanSpeed_V})
  ) {speedX=-3, speedY=0.2}
  n4 = TextureSample(UVs=@n3) {samplerType=SAMPLERTYPE_Normal, texture=/Game/Effects/Fort_Effects/Gameplay/Water/Textures/T_WaterRipples_01}
  AppendVector(A=LinearInterpolate(Alpha=@n4.R) {constA=-0.5}, B=LinearInterpolate(Alpha=@n4.G) {constA=-0.5})
  n5 = Multiply(A=VectorParameter() {parameterName=Fresnel Color, defaultValue=<rgba 0 1 0 1>}.RGB, B=Fresnel())
  n6 = ParticleColor()
  MaterialFunctionCall(Emissive=StaticSwitchParameter(True=Constant(), False=Multiply(A=@n5, B=@n6.R)) {parameterName=UseSplineMeshAsTrail}) {materialFunction=&MF_FX_EmissiveAdjustments}
  n7 = MaterialFunctionCall(UVs=Add(A=Constant2Vector() {r=-0.1}, B=TextureCoordinate()), Texture Scale=Constant2Vector() {g=1, r=4.5}) {materialFunction=&ScaleUVsByCenter}
  Multiply(
   A=TextureSample(UVs=@n7.UVs) {texture=/Game/Items/Art_noLOD/Textures/T_MedBandage_D}.RGB
   B=ScalarParameter() {defaultValue=1.63, parameterName=BaseColorBrightness}
  ) {constA=2, constB=1.63}
  Multiply(
   A=MaterialFunctionCall(
    Normal=TextureSample(UVs=@n7.UVs) {samplerType=SAMPLERTYPE_Normal, texture=/Game/Items/Art_noLOD/Textures/T_MedBandage_N}.RGB
    Flatness=Constant() {r=0.9}
   ) {materialFunction=&FlattenNormal}.Result
   B=ScalarParameter() {defaultValue=1, parameterName=Normal}
  ) {constB=0}
  n8 = ComponentMask(None=@n4.RGB) {g=true, r=true}
  n9 = LinearInterpolate(Alpha=TextureSample(UVs=@n3) {samplerType=SAMPLERTYPE_Grayscale, texture=&LowResBlurredNoise}.R)
  n10 = OneMinus(None=MaterialFunctionCall(Float2=TextureCoordinate()) {materialFunction=&BreakOutFloat2Components}.G)
  n11 = ScalarParameter() {defaultValue=2, group=WPO, parameterName=WPO_FalloffAlongEdge}
  n12 = Multiply(
   A=Multiply(A=AppendVector(A=@n8, B=@n9), B=ScalarParameter() {defaultValue=70, group=WPO, parameterName=WPO_Intensity}) {constB=20}
   B=Saturate(None=Multiply(A=@n10, B=@n11))
  )
  n13 = MaterialFunctionCall(Float2=TextureCoordinate()) {materialFunction=&BreakOutFloat2Components}
  n14 = VectorParameter() {parameterName=BulletOrigin}
  n15 = ActorPositionWS()
  n16 = Time() {period=5}
  n17 = ScalarParameter() {parameterName=StartTime}
  n18 = Clamp(None=Divide(A=Subtract(A=@n16, B=@n17), B=ScalarParameter() {defaultValue=0.2, group=Timestamp, parameterName=WPODuration}))
  n19 = Multiply(
   A=Multiply(A=OneMinus(None=@n13.G), B=Subtract(A=@n14.RGB, B=@n15))
   B=Clamp(None=MaterialFunctionCall!(In=@n18) {materialFunction=/Game/Effects/Fort_Effects/Materials/_Functions/MF_Smoothstep_EaseOut}.Result)
  )
  n20 = ScalarParameter() {defaultValue=15, parameterName=SplineWidthBase}
  n21 = ScalarParameter() {defaultValue=15, parameterName=SplineWidthTip}
  n22 = TextureCoordinate() {uTiling=0.2, vTiling=0.4}
  n23 = Time()
  n24 = DynamicParameter() {paramNames [Power, UV_Offset, "Erode(1-0)", ErodeItensity]}
  n25 = TextureSample(UVs=Panner(Coordinate=@n22, Time=Add(A=@n23, B=@n24.UV_Offset)) {speedX=0.3, speedY=0.3}) {texture=/Game/Global/Textures/Default/Noise/3xPerlin_M}
  n26 = AppendVector(A=ComponentMask(None=@n25.RGB) {g=true, r=true}, B=Multiply(A=@n25.B) {constB=-3})
  n27 = MaterialFunctionCall(Float2=@n22) {materialFunction=&BreakOutFloat2Components}
  n28 = StaticSwitchParameter(
   True=Add(
    A=Add(A=@n12, B=@n19)
    B=MaterialFunctionCall(WidthBase=@n20, WidthTip=@n21) {materialFunction=&SplineThicken}.WorldPositionOffset
   )
   False=Multiply(A=Multiply(A=@n26, B=@n27.G)) {constB=80}
  ) {parameterName=UseSplineMeshAsTrail}
  n29 = MaterialFunctionCall!(InVector=@n28) {materialFunction=/Game/Athena/Environments/Landscape/Foliage/MF_MakePreciseVec3}
  ShadingPathSwitch(Default=@n29.Result, Deferred=@n29.Result, Mobile=@n28) {desc=Migrated from FeatureLevelSwitch}
  n30 = TextureSample() {texture=&AtmosphericCloudNoise01}
  n31 = Multiply(
   A=MaterialFunctionCall(Float2=TextureCoordinate()) {materialFunction=&BreakOutFloat2Components}.G
   B=ScalarParameter() {defaultValue=5, parameterName=TrailHeadErode}
  )
  n32 = DynamicParameter() {paramNames [Power, UV_Offset, "Erode(1-0)", ErodeItensity]}
  n33 = Multiply(A=Subtract(B=@n32."Erode(1-0)"), B=@n32.ErodeItensity)
  n34 = VectorParameter() {parameterName=DistortUV_ctrl, defaultValue=<rgba 1 0.5 0 0>}
  n35 = Panner(Coordinate=TextureCoordinate() {coordinateIndex=1}, Speed=AppendVector(A=@n34.B, B=@n34.B))
  n36 = AppendVector(A=@n34.R, B=@n34.R)
  n37 = TextureSample(UVs=Add(A=@n32.UV_Offset, B=Multiply(A=@n35, B=@n36))) {texture=&AtmosphericCloudNoise01}
  n38 = LinearInterpolate(A=Subtract(B=@n33) {constA=0}, B=Subtract(B=@n33), Alpha=@n37.R)
  n39 = TextureCoordinate() {coordinateIndex=1}
  n40 = MaterialFunctionCall(Float2=Sine(None=Multiply(A=@n39) {constB=0.5})) {materialFunction=&BreakOutFloat2Components}
  n41 = Multiply(A=@n40.R, B=StaticSwitchParameter(True=ComponentMask(None=@n39) {g=true}, False=@n40.G) {parameterName=ForTrail})
  n42 = Multiply(A=@n41, B=LinearInterpolate(A=@n37.R, B=@n41, Alpha=Power(Base=@n41, Exp=@n32.Power) {constExponent=5}))
  StaticSwitchParameter(
   True=Saturate(None=Add(A=@n30.R, B=@n31))
   False=Add(A=LinearInterpolate(A=@n38, B=@n42, Alpha=@n32."Erode(1-0)"), B=ScalarParameter() {parameterName=EnhanceErode})
  ) {parameterName=UseSplineMeshAsTrail}
  TextureSample(UVs=@n3) {texture=&3xPerlin_M}
 ]
}
